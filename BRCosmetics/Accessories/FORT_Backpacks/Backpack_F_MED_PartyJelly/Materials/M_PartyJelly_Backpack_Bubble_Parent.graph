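# Translucent, lit parent material for the PartyJelly backpack's bubble, with front and back colors and time-of-day tuning.
# Vertex color separates inner from outer faces, blending Fresnel-driven opacities and masking detail normals and fake GGX sun specular.
# Panning water normals use noise-distorted two-plane projection on pre-skinned position, so they stick to the animated mesh.
# Sine-and-noise world position offset wobbles vertices along the direction to a hardcoded bulb position.
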Material /BRCosmetics/Accessories/FORT_Backpacks/Backpack_F_MED_PartyJelly/Materials/M_PartyJelly_Backpack_Bubble_Parent
graph {
 properties {
  bAutomaticallySetUsageInEditor=false
  bUsedWithMeshDeformer=true
  bUsedWithSkeletalMesh=true
  blendMode=BLEND_Translucent
  translucencyLightingMode=TLM_VolumetricDirectional
  usedShadingModels=Default Lit
 }
 nodes [
  n1 = SkyAtmosphereLightDirection()
  n2 = MaterialFunctionCall(Float3=@n1) {materialFunction=&BreakOutFloat3Components}
  n3 = DotProduct(
   A=MaterialFunctionCall(X=@n2.R, Y=@n2.R, Z=Constant() {r=1}) {materialFunction=&MakeFloat3}.Result
   B=ReflectionVectorWS()
  )
  n4 = ScalarParameter() {defaultValue=0.88, parameterName=SpecMult}
  Clamp(None=Power(Base=Multiply(A=@n3, B=@n4), Exp=ScalarParameter() {defaultValue=51, parameterName=Spec Highlight Power}) {constExponent=20})
  CollectionParameter() {
   collection=&FortniteMaterialParameters
   parameterName=SunAndMoonModelDirectionalVector
  }
  n5 = VectorParameter() {parameterName=Front Color, defaultValue=<rgba 0 1 0.7333 1>}
  n6 = VectorParameter() {parameterName=Back Color, defaultValue=<rgba 0.8296 1 0.1478 1>}
  LinearInterpolate(A=@n5.RGB, B=@n6.RGB, Alpha=VertexColor())
  n7 = TextureSample() {
   samplerType=SAMPLERTYPE_Normal
   texture=/BRCosmetics/Accessories/FORT_Backpacks/Backpack_F_MED_Angler/Skins/PartyJelly/Textures/T_PartyJelly_Backpack_N
  }
  MaterialFunctionCall(BaseNormal=@n7.RGB, AdditionalNormal=NamedRerouteUsage()) {materialFunction=&BlendAngleCorrectedNormals}
  n8 = Add(A=NamedRerouteUsage(), B=NamedRerouteUsage())
  n9 = NamedRerouteUsage()
  n10 = Constant3Vector() {constant=<rgba 1 1 1 0>}
  Multiply(
   A=Add(A=Add(A=@n8, B=@n9), B=NamedRerouteUsage())
   B=MaterialFunctionCall(Color_Morning=@n10, Color_Day=@n10, Color_Evening=@n10, Color_Night=Constant3Vector() {constant=<rgba 0.5 0.5833 1 0>}) {materialFunction=&Color_Time_of_Day}.Result
  )
  n11 = PixelNormalWS()
  n12 = ScalarParameter() {defaultValue=0.3, parameterName=Fake Spec Roughness, sliderMax=1}
  n13 = Power(
   Base=Multiply(
    A=MaterialFunctionCall!(Normal=@n11, Light Vector=@n1, Roughness=@n12) {materialFunction=/Engine/Functions/Engine_MaterialFunctions01/Lighting/GGXSpecular}.Result
    B=ScalarParameter() {defaultValue=2, parameterName=Spec Intense}
   )
   Exp=ScalarParameter() {defaultValue=1, parameterName=Spec Pow}
  )
  n14 = Constant() {r=1}
  n15 = MaterialFunctionCall(Value_Morning=@n14, Value_Day=@n14, Value_Evening=@n14, Value_Night=Constant() {r=0.1}) {materialFunction=&Scalar_Time_of_Day}
  n16 = VertexColor()
  NamedRerouteDeclaration(None=Multiply(A=Multiply(A=@n13, B=@n15.Result), B=OneMinus(None=@n16)) {desc=Mask out inner faces}) {name=Fake Spec, nodeColor=<rgba 0.4062 0 1 1>}
  n17 = Constant3Vector() {desc=Bulb Local Pos, constant=<rgba 0 -10 6 0>}
  n18 = Normalize(VectorInput=Subtract(A=TransformPosition(None=@n17) {transformType=TRANSFORMPOSSOURCE_World}, B=WorldPosition().XYZ))
  n19 = Time()
  n20 = Sine(None=Frac(None=Multiply(A=@n19) {constB=0.6}))
  n21 = LocalPosition()
  n22 = Subtract(A=Frac(None=Multiply(A=@n21.XYZ) {constB=0.01})) {constB=0.5}
  n23 = MaterialFunctionCall(Float3=Abs(None=Multiply(A=@n22) {constB=2})) {materialFunction=&BreakOutFloat3Components}
  n24 = Subtract(A=Add(A=Add(A=@n23.R, B=@n23.R), B=@n23.B)) {constB=2}
  n25 = Frac(None=Multiply(A=Time()) {constB=0.25})
  n26 = Add(A=Sine(None=Subtract(A=@n24, B=@n25)))
  n27 = Add(
   A=Multiply(A=Add(A=@n20)) {constB=0.5}
   B=Multiply(A=Multiply(A=@n26) {constB=0.5}, B=ScalarParameter() {defaultValue=1, parameterName=WPO_WorldNoiseScale})
  )
  NamedRerouteDeclaration(None=Multiply(A=Multiply(A=@n18, B=@n27), B=ScalarParameter() {defaultValue=1, parameterName=WPO_FinalScale})) {name=WPO, nodeColor=<rgba 0 1 0.0859 1>}
  n28 = TextureSample() {texture=/BRCosmetics/Accessories/FORT_Backpacks/Backpack_F_MED_Angler/Skins/PartyJelly/Textures/T_PartyJelly_Backpack_E}
  n29 = Constant() {r=5}
  NamedRerouteDeclaration(None=Multiply(A=@n5.RGB, B=Multiply(A=@n28.RGB, B=@n29) {constB=5})) {name=Base Emissive, nodeColor=<rgba 1 0 0.7969 1>}
  n30 = Fresnel(ExponentIn=ScalarParameter() {defaultValue=3, parameterName=Front Fresnel Exp}, Normal=NamedRerouteUsage())
  n31 = ScalarParameter() {defaultValue=0.5, parameterName=Front Max Opacity}
  n32 = Add(A=@n30, B=ScalarParameter() {defaultValue=0.05, parameterName=Front Min Opacity}) {constB=0.1}
  n33 = NamedRerouteDeclaration(None=Add(A=@n30, B=Multiply(A=@n31, B=@n32))) {name=Front Opacity, nodeColor=<rgba 0 1 0.4844 1>}
  n34 = Fresnel(ExponentIn=ScalarParameter() {defaultValue=2, parameterName=Back Fresnel Exp})
  n35 = NamedRerouteDeclaration(None=Multiply(A=ScalarParameter() {defaultValue=0.5, parameterName=Back Max Opacity}, B=OneMinus(None=@n34))) {name=Back Opacity, nodeColor=<rgba 0.8047 0 1 1>}
  n36 = VertexColor()
  NamedRerouteDeclaration(None=Add(A=NamedRerouteUsage(), B=LinearInterpolate(A=@n33, B=@n35, Alpha=@n36) {constA=1})) {name=Opacity, nodeColor=<rgba 0 1 0.8828 1>}
  n37 = PreSkinnedPosition()
  n38 = Divide(A=@n37, B=ScalarParameter() {defaultValue=50, group=PanningEmissive, parameterName=Triplanner_ConstantScaleDivisor})
  n39 = Multiply(A=ScalarParameter() {defaultValue=0.5941, parameterName=Distortion Tiling}, B=@n38)
  n40 = TextureObjectParameter() {group=PanningEmissive, parameterName=EmissiveGradient, texture=/Game/Global/Textures/Default/Noise/3xPerlin_M}
  n41 = LinearInterpolate(
   A=TextureSample(UVs=ComponentMask(None=@n39) {b=true, r=true}, Tex=@n40) {samplerType=SAMPLERTYPE_Masks, texture=/Game/Weapons/WeaponSkins/Textures/T_Swirly_Noise_highContrast}.RGB
   B=TextureSample(UVs=ComponentMask(None=@n39) {b=true, g=true}, Tex=@n40) {samplerType=SAMPLERTYPE_Masks, texture=&T_Swirly_Noise_highContrast}.RGB
  )
  NamedRerouteDeclaration(None=Multiply(A=VertexInterpolator(VS=@n41).PS, B=ScalarParameter() {defaultValue=0.5, parameterName=Distortion Intensity}) {constB=0.01}) {name=Distortion, nodeColor=<rgba 1 0 0.2344 1>}
  n42 = ScalarParameter() {defaultValue=0.25, parameterName=WaterTiling}
  n43 = VertexInterpolator(VS=@n38)
  n44 = Multiply(A=Time(), B=VectorParameter() {group=PanningEmissive, parameterName=Water Speed, defaultValue=<rgba 0 0 0.01 1>}.RGB)
  n45 = Add(A=Add(A=Multiply(A=@n42, B=@n43.PS), B=Frac(None=@n44)), B=NamedRerouteUsage())
  n46 = TextureObjectParameter() {
   group=PanningEmissive
   parameterName=WaterNormal
   samplerType=SAMPLERTYPE_Normal
   texture=/Engine/Functions/Engine_MaterialFunctions02/ExampleContent/Textures/water_n
  }
  n47 = TextureSample(UVs=ComponentMask(None=@n45) {b=true, r=true}, Tex=@n46) {samplerType=SAMPLERTYPE_Normal, texture=&T_Swirly_Noise_highContrast}
  n48 = TextureSample(UVs=ComponentMask(None=@n45) {g=true, r=true}, Tex=@n46) {samplerType=SAMPLERTYPE_Normal, texture=&T_Swirly_Noise_highContrast}
  n49 = Abs(None=ComponentMask(None=PreSkinnedNormal()) {b=true})
  n50 = ScalarParameter() {defaultValue=0.6095, group=PanningEmissive, parameterName=TriPlanner_AxisBlendStrength}
  n51 = VertexInterpolator(VS=ComponentMask(None=MaterialFunctionCall(In=@n49, Contrast=@n50) {materialFunction=&CheapContrast}.Result) {b=true})
  n52 = Multiply(A=ScalarParameter() {defaultValue=0.75, parameterName=WaterIntensity}, B=OneMinus(None=VertexColor())) {desc=Mask out inner faces}
  NamedRerouteDeclaration(None=MaterialFunctionCall(Normal=LinearInterpolate(A=@n47.RGB, B=@n48.RGB, Alpha=@n51.PS), Flatness=OneMinus(None=@n52)) {materialFunction=&FlattenNormal}.Result) {name=Detail Normal, nodeColor=<rgba 0 0.5547 1 1>}
  NamedRerouteDeclaration(None=Transform(None=@n7.RGB)) {name=WorlNormal, nodeColor=<rgba 0.875 1 0 1>}
  n53 = NamedRerouteUsage()
  n54 = ScalarParameter() {defaultValue=0.35, parameterName=Front Color Intensity}
  NamedRerouteDeclaration(None=Multiply(A=Multiply(A=@n53, B=@n54) {constB=0.1}, B=@n5.RGB)) {name=Front Emissive, nodeColor=<rgba 1 0.1641 0 1>}
  n55 = ScalarParameter() {defaultValue=0.5, parameterName=Back Color Intensity - Night}
  n56 = MaterialFunctionCall(Value_Morning=@n55, Value_Day=ScalarParameter() {parameterName=Back Color Intensity - Day}, Value_Evening=@n55, Value_Night=@n55) {materialFunction=&Scalar_Time_of_Day}
  NamedRerouteDeclaration(None=Multiply(A=Multiply(A=@n6.RGB, B=@n56.Result) {constB=0.1}, B=NamedRerouteUsage())) {name=Back Emissive, nodeColor=<rgba 0 0.1562 1 1>}
  NamedRerouteDeclaration(None=Multiply(A=@n28.B) {constB=3}) {name=Base Opacity, nodeColor=<rgba 0.4766 1 0 1>}
  NamedRerouteUsage()
  NamedRerouteUsage()
  VertexInterpolator(VS=@n37)
 ]
}
